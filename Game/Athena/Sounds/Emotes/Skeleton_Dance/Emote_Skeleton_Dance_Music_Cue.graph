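# Music cue for the Skeleton Dance emote, using the first-person emote music sound class and concurrency rule.
# A mixer layers two delayed branches: a one-shot pickup that starts almost immediately and a looping track that follows half a second later, forming an intro into a loop.
# It also references a music audio-analysis submix send.
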SoundCue /Game/Athena/Sounds/Emotes/Skeleton_Dance/Emote_Skeleton_Dance_Music_Cue
root=Mixer_0
nodes [
 {id=Delay_0, children [WavePlayer_0], properties {delayMax=0.5, delayMin=0.5}}
 {id=Delay_1, children [WavePlayer_3], properties {delayMax=0.02, delayMin=0.02}}
 {id=Mixer_0, children [Delay_0,Delay_1], properties {inputVolume [1,1]}}
 {id=WavePlayer_0, properties {bLooping=true, soundWaveAssetPtr=&/Emote_Skeleton_Dance_Loop_01}}
 {id=WavePlayer_3, properties {soundWaveAssetPtr=&/Emote_Skeleton_Dance_Pickup_01}}
]
properties {
 maxDistance=2097152
 soundClassObject=/Game/Sounds/FortSoundClassesAndModes/Fort_Music_Emote_1P_PSM
 volumeMultiplier=1.1
 concurrencyOverrides {maxCount=3, resolutionRule=StopOldest}
 concurrencySet [/Game/Sounds/Concurrency_Rules/Athena_EmoteMusic_1P_Conc]
 soundSubmixSends [{minSendLevel=0, soundSubmix=&Fortnite_AudioAnalysis_Music}]
}
